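; Quest ability that reports progress for dealing at least 150 critical-hit damage with the two-shot pump shotgun.
; Its response filter accepts critical hits on non-teammate player pawns or any other Fort pawn.
; Activation accumulates event magnitude into DamageDone, listens for further shotgun events, and sends the quest stat and ends at the threshold.
; A Reload event, bound to a timer with default time, waits for a tagged event to end the ability.
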(event Ability|EventActivateAbilityFromEvent
  (params (EventData :type Gameplay%20Event%20Data%20Structure%20%28by%20ref%29))
  (bind (_eventtag _instigator _target _optionalobject _optionalobject2 _contexthandle _instigatortags _targettags _eventmagnitude) (Utilities|Struct|BreakGameplayEventData EventData))
  ;[1 1 1] Activate ability - listen for reload while using shotgun
  (|CallParentFunction EventData)
  (if (GameplayTags|HasTag _instigatortags "(TagName=\"Item.Weapon.Ranged.Shotgun.Pump.TwoShot\")")
    (Variables|Default|SetDamageDone (+ _eventmagnitude (Variables|Default|GetDamageDone)))
    (bind _returnvalue (Utilities|Time|SetTimerbyEvent (delegate Custom|Reload)))
    (Variables|Default|SetReload%20Handle _returnvalue)
    (bind (_asynctaskproxy _payload) (Ability|Tasks|WaitGameplayEvent (Variables|Default|GetEvent%20Tag_0))
      (:then
        _)
      (:EventReceived
        (if (GameplayTags|HasTag (.InstigatorTags (Utilities|Struct|BreakGameplayEventData _payload)) "(TagName=\"Item.Weapon.Ranged.Shotgun.Pump.TwoShot\")")
          (Variables|Default|SetDamageDone (+ (.EventMagnitude (Utilities|Struct|BreakGameplayEventData _payload)) (Variables|Default|GetDamageDone))))))
    (if (>= (Variables|Default|GetDamageDone) 1.5e+02)
      (Quest|SendComplexCustomStatEvent (Variables|GAVars|Data|GetOwningPlayerQuestManager) 0 (Variables|Default|GetAdditional%20Source%20Tags) (.TargetTags (Utilities|Struct|BreakGameplayEventData _payload)) 1)
      (Ability|EndAbility))
    (else
      (Ability|EndAbility))))

(event Custom|Reload
  ;[1 1 1] Activate ability - listen for reload while using shotgun
  (Ability|Tasks|WaitGameplayEvent (Variables|Default|GetEvent%20Tag)
    (:EventReceived
      (Ability|EndAbility))))

(event Ability|EventOnEndAbility
  (params (bWasCancelled :type Boolean))
  (bind _reload_handle (Variables|Default|GetReload%20Handle))
  ;[1 1 1] End ability, clear timer if it is valid
  (if (Utilities|Time|IsValidTimerHandle _reload_handle)
    (Utilities|Time|ClearandInvalidateTimerbyHandle _reload_handle)))

(fn ShouldAbilityRespondToEvent
  (params (ActorInfo :type Gameplay%20Ability%20Actor%20Info%20Structure) (Payload_EventTag :type Gameplay%20Tag%20Structure) (Payload_Instigator :type Actor%20Object%20Reference) (Payload_Target :type Actor%20Object%20Reference) (Payload_OptionalObject :type Object%20Reference) (Payload_OptionalObject2 :type Object%20Reference) (Payload_ContextHandle :type Gameplay%20Effect%20Context%20Handle%20Structure) (Payload_InstigatorTags :type Gameplay%20Tag%20Container%20Structure) (Payload_TargetTags :type Gameplay%20Tag%20Container%20Structure) (Payload_EventMagnitude :type Float%20%28single-precision%29) (Payload_TargetData :type Gameplay%20Ability%20Target%20Data%20Handle%20Structure) (Payload_InstancedEventData :type Instanced%20Struct%20Structure))
  (bind _returnvalue (Ability|EffectContext|IsCriticalHit Payload_ContextHandle))
  (bind _asfort_player_pawn (Utilities|Casting|CastToFortPlayerPawn Payload_Target)
    (:then
      (return (and (not (Team|OnSameTeam _asfort_player_pawn Payload_Instigator)) _returnvalue)))
    (:CastFailed
      (Utilities|Casting|CastToFortPawn Payload_Target
        (:then
          (if _returnvalue
            (return true)
            (else
              (return false))))
        (:CastFailed
          (return false))))))
